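# Sound cue for the reaction sound of the Basil back bling cosmetic.
# A fixed short delay feeds an equally weighted random node that picks one of three reaction wave variations.
# It plays at reduced volume through the backblings sound class, using a close-range back bling attenuation and a reactive back bling audio property sheet.
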SoundCue /BRCosmetics/Sounds/Backblings/Basil/Backpack_Basil_React_Cue
root=Delay_0
nodes [
 {id=Delay_0, children [Random_0], properties {delayMax=0.2, delayMin=0.2}}
 {id=Random_0, children [WavePlayer_2,WavePlayer_3,WavePlayer_4], properties {weights [1,1,1]}}
 {id=WavePlayer_2, properties {soundWaveAssetPtr=&/Backbling_Basil_React_03}}
 {id=WavePlayer_3, properties {soundWaveAssetPtr=&/Backbling_Basil_React_01}}
 {id=WavePlayer_4, properties {soundWaveAssetPtr=&/Backbling_Basil_React_02}}
]
properties {
 attenuationSettings=/BRCosmetics/Sounds/Other/Attenuation/Fort_BackBling_Close_Att
 audioPropertiesSheet=/BRCosmetics/Sounds/Settings/PropertySheets/Reactive/APS_Cosmetic_Reactive_Backbling
 maxDistance=1111.11
 soundClassObject=/Game/Sounds/FortSoundClassesAndModes/Fort_Backblings
 volumeMultiplier=0.63
}
